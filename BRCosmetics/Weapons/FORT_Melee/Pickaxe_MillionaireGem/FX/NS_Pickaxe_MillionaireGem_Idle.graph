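# Niagara system that adds an idle sparkle effect to the MillionaireGem pickaxe.
# A single Sparkle sprite emitter spawns a few short-lived particles intermittently, placing each at a point selected from a baked position array instead of sampling a mesh.
# Curves pop size and alpha in, then fade them out, while a camera offset shifts the sprites in depth.
# A time-of-day scalar dims their color at evening and night.
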NiagaraSystem /BRCosmetics/Weapons/FORT_Melee/Pickaxe_MillionaireGem/FX/NS_Pickaxe_MillionaireGem_Idle
dependencies {
 usedDataInterfaces [&NiagaraDataInterfaceArrayFloat3, &NiagaraDataInterfaceCurve]
 usedDynamicInputs [&Multiply_Vector2D_ByFloat]
 usedModules [&SystemState]
}
emitters {Sparkle {
  emitterUpdateScript [
   &EmitterState()
   &SpawnRate(Spawn Probability=0.5, SpawnRate=5, Use Spawn Probability=true)
  ]
  particleSpawnScript [
   &InitializeParticle(
    Lifetime=&RandomRangeFloat(Minimum=0.4, Maximum=0.8)
    Sprite Rotation Angle=0.25
    Sprite Rotation Mode=::NewEnumerator1
    Sprite Size Mode=::NewEnumerator1
    Uniform Sprite Size Max=15
    Uniform Sprite Size Min=10
    UsePositionOffset=false
    Write Material Random=true
   )
   /SetVariables(Particles.Position=&SelectVectorFromArray(
     Array Sampling Mode=::NewEnumerator0
     Direct Array Index=?
     Vector Selection Array=DataInterface() {propertyValues {FloatData [
        <0.7327 -30.9732 48.5454>
        <0.6563 -28.9053 48.8476>
        <0.4337 -32.3725 48.4066>
        <-1.1989 -21.5322 49.831>
        <-1.9315 -17.5366 50.2576>
        <-1.9412 -16.5547 50.5056>
        <-1.5478 -14.7471 51.3006>
        <-0.6983 -24.6487 48.4208>
        <-2.0071 -14.2791 49.0862>
        <-0.7817 -22.4753 48.1167>
        <-0.9958 -20.6792 48.2027>
        <-1.0929 -21.7219 48.6672>
        <-1.8894 -18.141 49.6975>
        <-2.4666 -11.1647 51.3079>
        <-2.2497 -12.948 49.3417>
        <-2.6072 -8.3235 51.2137>
        <-1.8337 -6.9101 51.1013>
        <-2.6494 -7.6767 50.8484>
        <-1.6153 -7.1633 49.0248>
        <-2.7192 -8.7614 49.4833>
        <-0.6981 -7.7353 48.1753>
        <-0.9504 -7.6362 48.3581>
        <-1.6895 -12.1959 48.1404>
        <-0.8264 -8.7876 47.7229>
        <-0.1807 -24.1894 47.793>
        <-0.3137 -14.4833 47.6327>
        <0.6489 -23.2636 47.4374>
        <-0.0726 -20.3065 47.6637>
        <0.661 -24.8366 47.4672>
        <0.6929 -14.2949 47.2218>
        <0.0287 -12.8804 47.4593>
        <3.108 -7.2173 48.9329>
        <1.3427 -12.1897 47.3732>
        <2.1445 -9.163 47.6365>
        <1.4082 -27.0382 47.7393>
        <2.2929 -21.5434 47.9703>
        <2.2113 -17.3765 47.8414>
        <2.9137 -16.1904 48.0979>
        <2.0867 -15.0084 47.7366>
        <2.2575 -14.2745 47.7887>
        <4.0966 -8.5117 49.2228>
        <3.7841 -7.8073 49.5802>
        <4.0865 -9.9874 49.3367>
        <3.244 -16.6478 48.9258>
        <1.5014 -30.8472 48.2242>
        <3.1789 -20.095 49.5866>
        <2.2036 -23.1759 48.0478>
        <2.0148 -27.632 48.659>
        <3.6997 -15.8371 49.7929>
        <4.4332 -12.2337 50.6465>
        <4.2957 -11.8611 50.2394>
        <4.6459 -10.504 50.9958>
        <2.8624 -20.438 50.0523>
        <1.833 -27.5617 49.0263>
        <3.9052 -14.0323 50.82>
        <2.0786 -22.4073 50.1607>
        <1.075 -29.8963 48.811>
        <0.6745 43.2254 48.8469>
        <-0.3703 37.7191 49.788>
        <-0.6944 32.7229 50.8258>
        <0.4592 47.9886 48.0601>
        <-0.9902 30.996 51.1416>
        <-1.3112 35.3362 49.7745>
        <-0.8685 36.6062 49.7284>
        <-1.7178 30.5155 50.6941>
        <0.436 48.048 47.8674>
        <-1.0151 38.0734 48.953>
        <-2.2796 30.5931 50.1773>
        <-0.1201 43.7982 48.185>
        <-0.4165 40.531 48.1251>
        <-2.0345 30.8587 49.6688>
        <-0.5017 40.722 48.3669>
        <-3.2947 24.0281 51.004>
        <-1.6496 21.323 49.4007>
        <-2.5645 22.0736 50.6312>
        <-2.6434 25.0702 49.7401>
        <-2.0656 26.6158 48.7621>
        <-0.5579 19.9933 49.5979>
        <-0.7174 23.4851 47.6862>
        <0.2399 20.3237 49.0012>
        <0.7923 20.5794 48.7796>
        <1.1183 20.2219 49.0164>
        <-0.2256 38.9581 47.8212>
        <0.4617 29.0832 47.3257>
        <-0.155 34.1456 47.6865>
        <0.6457 34.0774 47.3614>
        <-1.5035 27.4947 48.0847>
        <0.6034 38.7928 47.4825>
        <0.6939 26.3367 47.1727>
        <0.728 48.859 47.651>
        <2.4907 21.7017 48.5319>
        <1.006 23.8597 47.1825>
        <2.1224 25.1102 47.6596>
        <2.2985 23.3264 47.696>
        <1.7361 35.8486 47.7451>
        <1.9807 36.9631 47.8696>
        <0.9378 33.0853 47.359>
        <1.7287 34.9354 47.7211>
        <1.8039 40.066 47.8696>
        <1.1202 36.049 47.5009>
        <3.6749 22.5387 48.5526>
        <3.5894 26.6715 48.7198>
        <4.1553 23.6793 49.3656>
        <4.1556 21.8153 51.0067>
        <3.1577 30.6613 48.6545>
        <3.2995 29.1162 48.62>
        <3.5545 28.7371 49.1226>
        <2.9065 35.4082 49.1892>
        <3.2629 33.5368 49.5759>
        <3.721 30.2936 49.8756>
        <4.4709 24.4873 50.2585>
        <4.1155 26.0442 49.7956>
        <4.1644 26.0234 51.0908>
        <2.5239 38.8604 49.1136>
        <2.8645 34.3568 49.9474>
        <1.5283 44.0688 48.5882>
        <3.1398 29.3105 51.1357>
        <1.8187 37.0204 49.9692>
        <1.3755 38.815 49.7705>
        <0.9304 49.059 47.8705>
        <-0.5058 12.6748 56.1366>
        <0.9514 13.3499 56.7431>
        <-3.8368 11.1797 56.6276>
        <-1.5934 12.576 56.1878>
        <-0.3428 13.1817 58.5123>
        <0.0216 13.6553 58.2344>
        <-0.3698 8.9861 63.1062>
        <-2.3132 12.1564 58.327>
        <0.4088 12.4869 59.7766>
        <0.7768 10.5298 62.0905>
        <-4.5365 7.9714 54.9039>
        <-4.813 9.1298 56.5073>
        <-0.0291 7.2798 64.8611>
        <-4.1867 9.9519 58.2666>
        <-5.0032 8.9088 57.9061>
        <-2.3988 7.7839 61.7097>
        <-5.6662 8.1085 57.2919>
        <-5.9383 7.3388 57.5052>
        <-3.174 4.4514 59.5354>
        <-1.1914 6.9294 63.3976>
        <-4.9883 5.0713 57.7224>
        <-2.4407 5.4942 61.0392>
        <-5.7854 6.4655 57.3022>
        <-4.785 5.3445 56.3087>
        <-4.1926 5.305 55.2384>
        <-3.2555 3.6429 55.2425>
        <-0.9671 5.3046 62.6286>
        <-2.1749 1.818 58.0409>
        <-0.0434 3.7794 61.6166>
        <0.2662 6.1261 64.4024>
        <0.4289 6.2998 64.7016>
        <-2.9805 2.7711 58.552>
        <-1.2275 1.3582 56.6306>
        <1.8641 1.4106 56.346>
        <1.2665 1.8946 55.9112>
        <0.7311 2.2228 55.6163>
        <5.1321 9.8881 55.1129>
        <4.1638 12.0295 56.1407>
        <2.5151 10.8589 60.6157>
        <1.5926 13.2093 58.6465>
        <1.749 10.1418 61.9141>
        <1.7421 11.288 60.6589>
        <1.4696 11.6586 60.4331>
        <6.4693 8.207 55.9397>
        <5.8631 9.0371 55.5853>
        <6.5406 7.5063 58.6934>
        <3.7113 8.1008 61.7426>
        <4.6771 8.65 60.2534>
        <5.0086 8.3353 60.0422>
        <6.8146 6.963 55.8222>
        <4.0266 5.3756 60.8613>
        <2.8554 6.5094 62.9461>
        <2.8709 5.6483 62.4142>
        <4.0921 6.5376 61.4753>
        <4.423 5.231 60.2983>
        <5.5616 5.5141 54.8035>
        <5.8073 5.4658 55.308>
        <4.7201 3.3288 55.4824>
        <2.9177 4.0751 61.0242>
        <4.552 2.5784 58.2904>
        <0.9519 4.6447 62.9595>
        <1.4972 2.7373 60.4997>
        <1.8203 0.5512 57.4181>
       ]}}
     Random Seed=0
    )) {isSetParameters=true}
   &CameraOffset(Camera Offset Amount=5)
  ]
  particleUpdateScript [
   &ParticleState()
   &ScaleSpriteSize(Uniform Curve Sprite Scale=DataInterface() {propertyValues {Curve {keys [
        {arriveTangent=19.1684, interpMode=RCIM_Cubic, leaveTangent=19.1684, tangentMode=RCTM_User, time=0, value=0}
        {arriveTangent=-1.7917, interpMode=RCIM_Cubic, leaveTangent=-1.7917, tangentMode=RCTM_User, time=0.2, value=1}
        {arriveTangent=-0.0405, interpMode=RCIM_Cubic, leaveTangent=-0.0405, tangentMode=RCTM_User, time=1, value=0}
       ]}}})
   &ScaleColor(
    Scale Alpha=&FloatFromCurve(FloatCurve=DataInterface() {propertyValues {Curve {keys [
         {arriveTangent=9.0725, interpMode=RCIM_Cubic, leaveTangent=9.0725, tangentMode=RCTM_User, time=0, value=0}
         {arriveTangent=-2.5249, interpMode=RCIM_Cubic, leaveTangent=-2.5249, tangentMode=RCTM_User, time=0.1408, value=0.9847}
         {arriveTangent=-0.0405, interpMode=RCIM_Cubic, leaveTangent=-0.0405, tangentMode=RCTM_User, time=1, value=0}
        ]}}})
    Scale RGB=&VectorFromFloat(Value=&FNTimeOfDayScalar(Night Scale=0.5, Morning Scale=1, Evening Scale=0.8))
    ScaleA=true
    ScaleRGB=true
   )
  ]
  rendererClasses [&NiagaraSpriteRendererProperties]
  renderers [{rendererClass=&NiagaraSpriteRendererProperties}]
 }}
summary {
 systemName=NS_Pickaxe_MillionaireGem_Idle
 emitters [{emitterName=Sparkle, rendererClasses [&NiagaraSpriteRendererProperties]}]
 userVariables [
  User.bIsFrontEnd {type=NiagaraBool}
  User.bIsFrontEndPreview {type=NiagaraBool}
 ]
}
