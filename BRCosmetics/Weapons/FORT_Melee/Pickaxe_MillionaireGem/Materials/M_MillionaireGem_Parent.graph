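# Parent material for the MillionaireGem pickaxe, built on material attributes for skeletal meshes with multiple shading models.
# It chains the 2023 character parent PreFX function into a gemstone function borrowed from the DiamondHeart Jewel character, then into the matching PostFX function.
# The GS_Mask channel parameter picks a channel of the PreFX Tech Art Mask to feed the gemstone mask input.
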Material /BRCosmetics/Weapons/FORT_Melee/Pickaxe_MillionaireGem/Materials/M_MillionaireGem_Parent
graph {
 properties {
  bUseMaterialAttributes=true
  bUsedWithMeshDeformer=true
  bUsedWithSkeletalMesh=true
  pixelDepthOffsetMode=PDOM_Legacy
  shadingModel=MSM_FromMaterialExpression
  usedShadingModels=Default Lit | Subsurface | Clear Coat | Subsurface Profile | Cloth
 }
 nodes [
  n1 = MaterialFunctionCall() {materialFunction=&MF_CharacterParent_2023_PreFX}
  MaterialFunctionCall(
   Material=MaterialFunctionCall!(
    Material=@n1.Result
    Mask Texture=@n1.Result
    TechArtMask=ChannelMaskParameter(None=@n1.Tech Art Mask) {group=Gemstone, parameterName=GS_Mask}.R
   ) {materialFunction=/Game/Characters/Player/Male/Medium/Bodies/M_MED_DiamondHeart_Jewel/Material/MF_Gemstone}.Result
   M=@n1.M
   Pre Skinned Local Position=@n1.M
   Distance Blend=@n1.M
   Fresnel=@n1.M
  ) {materialFunction=&MF_CharacterParent_2023_PostFX}
 ]
}
